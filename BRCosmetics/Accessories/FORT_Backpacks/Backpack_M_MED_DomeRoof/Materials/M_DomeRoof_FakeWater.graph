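# Character material for the Dome Roof backpack that fakes animated water with foam around a ship.
# It overrides base color, normal, emissive and world position offset between the CharacterParent PreFX and PostFX functions.
# Panning normal and height maps drive masked ripples and displacement, with Fresnel-tinted emissive.
# Foam uses planar UVs centered on ShipCenter and aligned to local axes, with distortion, time-of-day intensity and a checker debug switch.
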Material /BRCosmetics/Accessories/FORT_Backpacks/Backpack_M_MED_DomeRoof/Materials/M_DomeRoof_FakeWater
graph {
 properties {
  bUseMaterialAttributes=true
  bUsedWithMeshDeformer=true
  bUsedWithSkeletalMesh=true
  pixelDepthOffsetMode=PDOM_Legacy
  shadingModel=MSM_FromMaterialExpression
  usedShadingModels=Default Lit | Subsurface | Clear Coat | Subsurface Profile | Cloth
 }
 nodes [
  n1 = MaterialFunctionCall() {materialFunction=&MF_CharacterParent_2023_PreFX}
  n2 = GetMaterialAttributes(None=@n1.Result) {attributeGetTypes [#1, #2]}
  n3 = Transform(None=NamedRerouteUsage())
  n4 = Add(
   A=Multiply(
    A=VectorParameter() {group=WaterSimulation, parameterName=EMColorOverride, defaultValue=<rgba 0 0.8385 0.6657 1>}.RGB
    B=Fresnel(Normal=@n3) {baseReflectFraction=0, exponent=1}
   )
   B=NamedRerouteUsage()
  )
  n5 = LinearInterpolate(B=NamedRerouteUsage(), Alpha=NamedRerouteUsage()) {constA=1}
  n6 = SetMaterialAttributes!(
   MaterialAttributes=@n2.MaterialAttributes
   Normal=NamedRerouteUsage()
   Base Color=VectorParameter() {group=WaterSimulation, parameterName=BaseColorOverride, defaultValue=<rgba 0.0999 0.1845 0.1912 1>}.RGB
   Emissive Color=Add(A=NamedRerouteUsage(), B=Multiply(A=@n4, B=@n5))
   World Position Offset=NamedRerouteUsage()
  ) {attributeSetTypes [
    #1
    #3
    #2
    #4
   ]}
  MaterialFunctionCall(Material=@n6, M=@n1.M, Pre Skinned Local Position=@n1.M, Distance Blend=@n1.M, Fresnel=@n1.M) {materialFunction=&MF_CharacterParent_2023_PostFX}
  n7 = TextureSampleParameter2D() {
   group=WaterSimulation
   parameterName=TechArtMask_RE
   samplerType=SAMPLERTYPE_Masks
   texture=/BRCosmetics/Accessories/FORT_Backpacks/Backpack_M_MED_DomeRoof/Textures/T_DomeRoof_Backpack_GlassMask
  }
  NamedRerouteDeclaration(None=@n7.G) {name=HeightDisplaceArea, nodeColor=<rgba 0 1 0.9531 1>}
  n8 = Panner(
   Coordinate=Multiply(A=TextureCoordinate(), B=ScalarParameter() {defaultValue=3, parameterName=WaterTilling})
   Speed=AppendVector(
    A=ScalarParameter() {defaultValue=0.12, group=WaterSimulation, parameterName=WaterSpeedX}
    B=ScalarParameter() {group=WaterSimulation, parameterName=WaterSpeedY}
   )
  ) {speedX=0.1}
  n9 = TextureSampleParameter2D(UVs=@n8) {
   group=WaterSimulation
   parameterName=WaterHeight_N
   samplerType=SAMPLERTYPE_Normal
   texture=/BRCosmetics/Accessories/FORT_Backpacks/Backpack_M_MED_DomeRoof/Textures/T_WaterPattern_N
  }
  NamedRerouteDeclaration(None=@n9.RGB) {name=WaterN, nodeColor=<rgba 0.0078 1 0 1>}
  n10 = Transform(None=Constant3Vector() {constant=<rgba 1 0 0 0>}) {transformSourceType=TRANSFORMSOURCE_Local}
  n11 = WorldPosition()
  n12 = VectorParameter() {group=WaterSimulation_Foam, parameterName=ShipCenter, defaultValue=<rgba 0 0 0 1>}
  n13 = Divide(
   A=ComponentMask(None=Subtract(A=@n11.XYZ, B=@n12.RGB)) {g=true, r=true}
   B=ScalarParameter() {defaultValue=10, group=WaterSimulation_Foam, parameterName=ShipRadius}
  )
  n14 = CrossProduct(A=@n10, B=Constant3Vector() {constant=<rgba 0 0 1 0>})
  n15 = Add(A=AppendVector(A=DotProduct(A=@n10, B=@n13), B=DotProduct(A=@n13, B=@n14)), B=Constant2Vector() {g=0.5, r=0.5})
  n16 = ComponentMask(None=@n15) {r=true}
  n17 = Constant() {r=1}
  n18 = Constant()
  n19 = Constant() {r=1}
  n20 = ComponentMask(None=@n15) {g=true}
  n21 = Constant() {r=1}
  n22 = Constant()
  n23 = Constant() {r=1}
  n24 = Multiply(
   A=Multiply(A=If(A=@n16, "A > B"=@n17, "A == B"=@n18, "A < B"=@n18) {constB=0.01}, B=If(A=@n16, "A > B"=@n18, "A == B"=@n18, "A < B"=@n19) {constB=0.99})
   B=Multiply(A=If(A=@n20, "A > B"=@n21, "A == B"=@n22, "A < B"=@n22) {constB=0.01}, B=If(A=@n20, "A > B"=@n22, "A == B"=@n22, "A < B"=@n23) {constB=0.99})
  )
  NamedRerouteDeclaration(None=@n24) {name=UVMask, nodeColor=<rgba 0 1 0.5078 1>}
  n25 = TextureSampleParameter2D(UVs=@n8) {
   group=WaterSimulation
   parameterName=WaterHeight_H
   samplerType=SAMPLERTYPE_LinearGrayscale
   texture=/BRCosmetics/Accessories/FORT_Backpacks/Backpack_M_MED_DomeRoof/Textures/T_WaterPattern_Height
  }
  n26 = Power(Base=@n25.R) {constExponent=2.2}
  n27 = Multiply(
   A=NamedRerouteDeclaration(None=Saturate(None=@n26)) {name=HeightOrg, nodeColor=<rgba 0.7813 0 1 1>}
   B=Transform(None=Constant3Vector() {constant=<rgba 0 0 1 0>})
  )
  n28 = ScalarParameter() {defaultValue=1, group=WaterSimulation, parameterName=DisplaceHeight}
  NamedRerouteDeclaration(None=Multiply(A=NamedRerouteUsage(), B=Multiply(A=@n27, B=@n28))) {name=HeightOUT, nodeColor=<rgba 1 0 0.8906 1>}
  NamedRerouteDeclaration(None=LinearInterpolate(A=@n2.Normal, B=NamedRerouteUsage(), Alpha=NamedRerouteUsage())) {name=Normal_OUT, nodeColor=<rgba 0.9453 0 1 1>}
  NamedRerouteDeclaration(None=@n1.Pre skinned Local Position) {name=PreskinnedPos, nodeColor=<rgba 0.6172 1 0 1>}
  NamedRerouteDeclaration(None=@n2.Emissive Color) {name=EM_IN, nodeColor=<rgba 0 0.8594 1 1>}
  n29 = ScalarParameter() {group=WaterSimulation, parameterName=FoamSpeedX}
  n30 = ScalarParameter() {defaultValue=0.1, group=WaterSimulation, parameterName=FoamSpeedY}
  n31 = TextureSampleParameter2D(UVs=Panner(Coordinate=@n15, Speed=AppendVector(A=@n29, B=@n30))) {
   parameterName=Foam_DistortN
   samplerType=SAMPLERTYPE_Normal
   texture=/BRCosmetics/Accessories/FORT_Backpacks/Backpack_M_MED_DomeRoof/Textures/T_FoamDistort_N
  }
  n32 = ComponentMask(None=@n31.RGB) {g=true, r=true}
  n33 = ScalarParameter() {defaultValue=0.3, group=WaterSimulation_Foam, parameterName=FoamDistortIntensity}
  n34 = ComponentMask(None=NamedRerouteUsage()) {g=true, r=true}
  n35 = ScalarParameter() {defaultValue=0.05, group=WaterSimulation_Foam, parameterName=WaterDistortIntensity}
  NamedRerouteDeclaration(None=Add(A=Multiply(A=@n32, B=@n33), B=Multiply(A=@n34, B=@n35))) {name=DistortFoam, nodeColor=<rgba 1 0.4453 0 1>}
  n36 = Add(A=NamedRerouteUsage(), B=@n15)
  n37 = Multiply(A=TextureSample(UVs=@n36) {texture=/Game/FN_PublicTurntable/Turntable_Lowpoly/Content/BlueprintAssets/Checker}.RGB, B=NamedRerouteUsage())
  n38 = ScalarParameter() {group=WaterSimulation, parameterName=FoamPatternSpeedX}
  n39 = ScalarParameter() {defaultValue=0.05, group=WaterSimulation, parameterName=FoamPatternSpeedY}
  n40 = TextureSampleParameter2D(UVs=Panner(Coordinate=@n36, Speed=AppendVector(A=@n38, B=@n39))) {
   group=WaterSimulation_Foam
   parameterName=FoamMask
   samplerSource=SSM_Wrap_WorldGroupSettings
   samplerType=SAMPLERTYPE_Masks
   texture=/BRCosmetics/Accessories/FORT_Backpacks/Backpack_M_MED_DomeRoof/Textures/T_Foam_D
  }
  n41 = Multiply(A=@n40.G) {constB=1.5}
  n42 = ScalarParameter() {defaultValue=2, group=WaterSimulation_Foam, parameterName=FoamIntensity}
  n43 = Divide(A=@n42) {constB=3}
  n44 = TextureSampleParameter2D(UVs=@n36) {
   group=WaterSimulation_Foam
   parameterName=FoamMask
   samplerType=SAMPLERTYPE_Masks
   texture=&T_Foam_D
  }
  n45 = Power(Base=@n44.R, Exp=ScalarParameter() {defaultValue=2.5, group=WaterSimulation_Foam, parameterName=FoamMaskPow_Small}) {constExponent=2.5}
  n46 = ScalarParameter() {defaultValue=1, group=WaterSimulation_Foam, parameterName=FoamSmallIntensity}
  n47 = Power(Base=@n44.R, Exp=ScalarParameter() {defaultValue=1.5, group=WaterSimulation_Foam, parameterName=FoamMaskPow})
  n48 = ScalarParameter() {defaultValue=0.5, group=WaterSimulation_Foam, parameterName=FoamLargeIntensity}
  n49 = Saturate(None=Add(A=Multiply(A=@n45, B=@n46), B=Multiply(A=@n47, B=@n48)))
  n50 = Multiply(
   A=Multiply(
    A=Add(A=@n41, B=@n40.B)
    B=MaterialFunctionCall(Value_Morning=@n42, Value_Day=@n42, Value_Evening=@n43, Value_Night=@n43) {materialFunction=&Scalar_Time_of_Day}.Result
   )
   B=NamedRerouteDeclaration(None=Multiply(A=@n49, B=@n24)) {name=ShipFoamMask, nodeColor=<rgba 1 0.8203 0 1>}
  )
  NamedRerouteDeclaration(None=Multiply(A=StaticSwitchParameter(True=@n37, False=@n50) {group=DebugTexture, parameterName=DebugUV}, B=NamedRerouteUsage())) {name=ShipFoam, nodeColor=<rgba 0.7578 1 0 1>}
  TransformPosition(None=VectorParameter() {group=WaterSimulation_Foam, parameterName=ShipCenterOffset, defaultValue=<rgba 0 0 0 1>}.RGB) {transformType=TRANSFORMPOSSOURCE_World}
  Transform(None=Constant3Vector() {constant=<rgba 0 1 0 0>}) {transformSourceType=TRANSFORMSOURCE_Local}
 ]
}
