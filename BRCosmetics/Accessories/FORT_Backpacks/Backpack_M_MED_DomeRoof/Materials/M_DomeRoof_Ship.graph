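# Character material for the DomeRoof Ship back bling cosmetic, built on the shared 2023 character parent functions and adding a simple wind sway.
# A time-driven sine, remapped by bias and scale, pushes vertices along a local-space direction that the WindMask green channel blends between -X and +Y.
# The red channel masks that world position offset, which is written into the PreFX material attributes before they pass through PostFX.
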Material /BRCosmetics/Accessories/FORT_Backpacks/Backpack_M_MED_DomeRoof/Materials/M_DomeRoof_Ship
graph {
 properties {
  bUseMaterialAttributes=true
  bUsedWithMeshDeformer=true
  bUsedWithSkeletalMesh=true
  pixelDepthOffsetMode=PDOM_Legacy
  shadingModel=MSM_FromMaterialExpression
  usedShadingModels=Default Lit | Subsurface | Clear Coat | Subsurface Profile | Cloth
 }
 nodes [
  n1 = MaterialFunctionCall() {materialFunction=&MF_CharacterParent_2023_PreFX}
  n2 = Multiply(A=Time(), B=ScalarParameter() {defaultValue=4, group=WPO_Wind, parameterName=TimeScale})
  n3 = TextureSampleParameter2D() {
   group=WPO_Wind
   parameterName=WindMask
   samplerType=SAMPLERTYPE_Masks
   texture=/BRCosmetics/Accessories/FORT_Backpacks/Backpack_M_MED_DomeRoof/Textures/T_DomeRoof_Backpack_WPOMask
  }
  n4 = Transform(None=LinearInterpolate(A=Constant3Vector() {constant=<rgba -1 0 0 0>}, B=Constant3Vector() {constant=<rgba 0 1 0 0>}, Alpha=@n3.G)) {transformSourceType=TRANSFORMSOURCE_Local}
  n5 = Multiply(
   A=ConstantBiasScale(None=Sine(None=@n2))
   B=Multiply(A=ScalarParameter() {defaultValue=0.03, group=WPO_Wind, parameterName=DamplingScale}, B=Normalize(VectorInput=@n4))
  )
  MaterialFunctionCall(
   Material=SetMaterialAttributes!(MaterialAttributes=@n1.Result, World Position Offset=Multiply(A=@n5, B=@n3.R)) {attributeSetTypes [#1]}
   M=@n1.M
   Pre Skinned Local Position=@n1.M
   Distance Blend=@n1.M
   Fresnel=@n1.M
  ) {materialFunction=&MF_CharacterParent_2023_PostFX}
 ]
}
