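# Additive, unlit glass overlay material for the DomeRoof backpack cosmetic, usable on skeletal meshes.
# Base color tinted by texture alpha and an emissive map feed MF_OverlayTrans_Generic_Simple alongside a glass mask and roughness input.
# Glass-masked opacity adds dust, weighted by inverted roughness, to the function's result.
# A default-on static switch enables World Position Offset that pulls vertices toward the object position by HT_CrunchVerts to hide them.
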Material /BRCosmetics/Accessories/FORT_Backpacks/Backpack_M_MED_DomeRoof/Materials/M_Glass_Overlay
graph {
 properties {
  bAutomaticallySetUsageInEditor=false
  bUseMaterialAttributes=true
  bUsedWithMeshDeformer=true
  bUsedWithSkeletalMesh=true
  blendMode=BLEND_Additive
  pixelDepthOffsetMode=PDOM_Legacy
  shadingModel=MSM_Unlit
  usedShadingModels=Unlit
 }
 nodes [
  Constant()
  n1 = TextureSampleParameter2D() {
   group=00_Override
   parameterName=BaseColor_RE
   texture=/BRCosmetics/Accessories/FORT_Backpacks/Backpack_M_MED_DomeRoof/Textures/T_DomeRoof_Backpack_D
  }
  n2 = LinearInterpolate(
   A=@n1.RGB
   B=VectorParameter() {group=00_Override, parameterName=AdditiveBaseColor, defaultValue=<rgba 1 1 1 1>}.RGB
   Alpha=NamedRerouteDeclaration(None=@n1.A) {name=OpacityIN, nodeColor=<rgba 0 0.25 1 1>}
  )
  n3 = Constant()
  n4 = TextureSampleParameter2D() {
   group=00_Override
   parameterName=EmissiveColor_RE
   texture=/BRCosmetics/Accessories/FORT_Backpacks/Backpack_M_MED_DomeRoof/Textures/T_DomeRoof_Backpack_E
  }
  n5 = TextureSample() {
   samplerType=SAMPLERTYPE_Masks
   texture=/BRCosmetics/Accessories/FORT_Backpacks/Backpack_M_MED_DomeRoof/Textures/T_DomeRoof_Backpack_GlassMask
  }
  n6 = NamedRerouteUsage()
  n7 = GetMaterialAttributes(None=MaterialFunctionCall!(
    Material=SetMaterialAttributes!(MaterialAttributes, Base Color=@n2, Roughness=@n3, Emissive Color=@n4.RGB) {attributeSetTypes [#1, #2, #3]}
    TechArtMask=@n5.RGB
    RoughnessInput=OneMinus(None=@n6)
   ) {materialFunction=&/MF_OverlayTrans_Generic_Simple}.Result) {attributeGetTypes [#3, #4]}
  n8 = TextureSampleParameter2D() {
   group=00_Override
   parameterName=RoughnessRE
   samplerType=SAMPLERTYPE_Masks
   texture=/BRCosmetics/Accessories/FORT_Backpacks/Backpack_M_MED_DomeRoof/Textures/T_DomeRoof_Backpack_Glass_S
  }
  n9 = OneMinus(None=NamedRerouteDeclaration(None=@n8.G) {name=RoughnessIN, nodeColor=<rgba 0 0.8359 1 1>})
  n10 = TextureSampleParameter2D() {parameterName=DustMulti, texture=/BRCosmetics/Accessories/FORT_Backpacks/Backpack_M_MED_DomeRoof/Textures/T_DustMulti}
  n11 = Add(
   A=Multiply(A=NamedRerouteUsage(), B=ScalarParameter() {defaultValue=0.1, group=00_Override, parameterName=OpacityORG}) {constB=0.1}
   B=Multiply(A=ScalarParameter() {defaultValue=0.5, group=00_Override, parameterName=OpacityDust}, B=Multiply(A=@n9, B=@n10.R))
  )
  n12 = Subtract(A=ObjectPositionWS(), B=WorldPosition().XYZ)
  n13 = ScalarParameter() {parameterName=HT_CrunchVerts}
  SetMaterialAttributes!(
   MaterialAttributes=@n7.MaterialAttributes
   Emissive Color=@n7.MaterialAttributes
   Opacity=Add(A=@n7.Opacity, B=Multiply(A=@n11, B=@n5.R))
   World Position Offset=StaticSwitchParameter(True=Multiply(A=@n12, B=@n13), False=Constant()) {defaultValue=true, parameterName=bUseVertexHide}
  ) {attributeSetTypes [#3, #4, #5]}
 ]
}
