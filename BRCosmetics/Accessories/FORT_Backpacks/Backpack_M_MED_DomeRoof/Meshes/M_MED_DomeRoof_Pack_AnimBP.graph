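# Anim Blueprint for the DomeRoof back bling, a ship in a bottle with simulated water, ship and sails.
# Additive bone offsets and rotations feed chained Anim Dynamics, while update logic sets body-type, DBNO, skydive and gravity variables.
# Post-evaluation passes the ShipCenter socket location to materials.
# Opening a SpringBreak-tagged glider spawns glass-break Niagara and sets hide/crunch material parameters, which reset otherwise, on skydive start or parachute close.
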AnimBlueprint /BRCosmetics/Accessories/FORT_Backpacks/Backpack_M_MED_DomeRoof/Meshes/M_MED_DomeRoof_Pack_AnimBP
target_skeleton=&/M_MED_DomeRoof_Pack_Skeleton
graphs [
 AnimGraph {nodes [
   Output Pose(Result=Use cached pose 'DynamicsSailCache'()) {layerGroup=None}
   Save cached pose 'StartPoseCache'(Pose=Input Pose()) {cacheName=StartPoseCache}
   n1 = "Transform (Modify) Bone - Bone: pack"!(ComponentPose=Local To Component(LocalPose=Use cached pose 'StartPoseCache'()), Alpha, Translation="0,-8.3,0") -> (Pose) {
    translationMode=BMM_Additive
    translationSpace=BCS_BoneSpace
    alphaBoolBlend {blendOption=Linear}
    boneToModify {boneName=pack}
    translation {y=-8.3}
   }
   n2 = "Transform (Modify) Bone - Bone: pack"(ComponentPose=@n1, Translation="0,-6.2,4") {
    translationMode=BMM_Additive
    translationSpace=BCS_BoneSpace
    alphaBoolBlend {blendOption=Linear}
    boneToModify {boneName=pack}
    translation {y=-6.2, z=4}
   }
   n3 = "Transform (Modify) Bone - Bone: pack"(ComponentPose=@n2, Translation="0,1,0") {
    translationMode=BMM_Additive
    translationSpace=BCS_BoneSpace
    alphaBoolBlend {blendOption=Linear}
    boneToModify {boneName=pack}
    translation {y=1}
   }
   n4 = "Transform (Modify) Bone - Bone: pack"(ComponentPose=@n3, Translation="0,0.8,0") {
    translationMode=BMM_Additive
    translationSpace=BCS_BoneSpace
    alphaBoolBlend {blendOption=Linear}
    boneToModify {boneName=pack}
    translation {y=0.8}
   }
   n5 = "Transform (Modify) Bone - Bone: base_water_jnt"!(ComponentPose=@n4, Alpha=1, Rotation="0,60,0") -> (Pose) {
    rotationMode=BMM_Additive
    rotationSpace=BCS_BoneSpace
    alphaBoolBlend {blendOption=Linear}
    alphaScaleBiasClamp {bInterpResult=true}
    boneToModify {boneName=base_water_jnt}
    rotation {yaw=60}
   }
   n6 = "Transform (Modify) Bone - Bone: base_water_jnt"(ComponentPose=@n5, Alpha=1, Rotation="0,100,0") {
    rotationMode=BMM_Additive
    rotationSpace=BCS_BoneSpace
    alphaBoolBlend {blendOption=Linear}
    alphaScaleBiasClamp {bInterpResult=true}
    boneToModify {boneName=base_water_jnt}
    rotation {yaw=100}
   }
   Save cached pose 'TransformModifiesCache'(Pose=Component To Local(ComponentPose=@n6)) {cacheName=TransformModifiesCache}
   n7 = Anim Dynamics - Bone: dyn_water_jnt!(ComponentPose=Local To Component(LocalPose=Use cached pose 'TransformModifiesCache'()), Alpha=1, GravityOverride:Vector="0,0,-980") -> (Pose) {
    angularDampingOverride=0.99
    angularSpringConstant=33000
    bAngularSpring=true
    bOverrideAngularDamping=true
    bUseGravityOverride=true
    gravityScale=0.35
    simulationSpace=BoneRelative
    alphaBoolBlend {blendOption=Linear}
    boundBone {boneName=dyn_water_jnt}
    componentAppliedLinearAccClamp=<3000 3000 3000>
    componentLinearAccScale=<1 1 1>
    componentLinearVelScale=<1 1 1>
    gravityOverride {z=-980}
    physicsBodyDefinitions [{
      boundBone {boneName=dyn_water_jnt}
      boxExtents=<40 45 45>
      constraintSetup {angularLimitsMax=<0 0 89>, angularLimitsMin=<0 0 -30>, angularTarget=<1 0.1 0>}
      localJointOffset=<10 0 0>
     }]
    relativeSpaceBone {boneName=root}
   }
   n8 = Anim Dynamics - Bone: dyn_ship_jnt!(ComponentPose=@n7, Alpha=1, GravityOverride:Vector="0,0,-980") -> (Pose) {
    angularDampingOverride=1
    angularSpringConstant=10000
    bAngularSpring=true
    bOverrideAngularDamping=true
    bUseGravityOverride=true
    gravityScale=0.35
    linearDampingOverride=1
    linearSpringConstant=100
    simulationSpace=BoneRelative
    alphaBoolBlend {blendOption=Linear}
    alphaScaleBiasClamp {bInterpResult=true}
    boundBone {boneName=dyn_ship_jnt}
    componentLinearAccScale=<1 1 1>
    componentLinearVelScale=<1 1 1>
    gravityOverride {z=-980}
    physicsBodyDefinitions [{
      boundBone {boneName=dyn_ship_jnt}
      boxExtents=<25 25 25>
      constraintSetup {angularLimitsMax=<0 30 0>, angularLimitsMin=<0 -30 0>, angularTarget=<1 0 0>}
      localJointOffset=<0.55 0 0>
     }]
    relativeSpaceBone {boneName=dyn_water_jnt}
   }
   n9 = Anim Dynamics - Bone: dyn_ship_jnt(ComponentPose=@n8, Alpha=1, GravityOverride="0,0,-980") {
    angularDampingOverride=1
    angularSpringConstant=10000
    bLinearSpring=true
    bOverrideLinearDamping=true
    linearDampingOverride=1
    linearSpringConstant=20
    alphaBoolBlend {blendOption=Linear}
    alphaScaleBiasClamp {bInterpResult=true}
    boundBone {boneName=dyn_ship_jnt}
    componentAppliedLinearAccClamp=<5000 5000 5000>
    gravityOverride {z=-980}
    physicsBodyDefinitions [{
      boundBone {boneName=dyn_ship_jnt}
      boxExtents=<1 1 1>
      constraintSetup {angularLimitsMax=<0 0 0>, angularLimitsMin=<0 0 0>, angularTarget=<0 0 0>, linearAxesMax=<1 0 0>, linearAxesMin=<-2.5 0 0>}
      localJointOffset=<0 0 0>
     }]
    relativeSpaceBone {boneName=root}
   }
   Save cached pose 'DynamicsShipCache'(Pose=Component To Local(ComponentPose=@n9)) {cacheName=DynamicsShipCache}
   n10 = Anim Dynamics - Bone: dyn_sail_0_jnt!(
    ComponentPose=Local To Component(LocalPose=Use cached pose 'DynamicsShipCache'())
    Alpha=1
    GravityOverride:Vector="0,0,-980"
    ExternalForce="0,0,0"
   ) -> (Pose) {
    angularDampingOverride=1
    angularSpringConstant=20000
    bAngularSpring=true
    bOverrideAngularDamping=true
    bUseGravityOverride=true
    gravityScale=0.35
    simulationSpace=RootRelative
    alphaBoolBlend {blendOption=Linear}
    boundBone {boneName=dyn_sail_0_jnt}
    componentLinearAccScale=<1 1 1>
    componentLinearVelScale=<1 1 1>
    gravityOverride {z=-980}
    physicsBodyDefinitions [{
      boundBone {boneName=dyn_sail_0_jnt}
      boxExtents=<30 30 30>
      constraintSetup {angularLimitsMax=<15 15 15>, angularLimitsMin=<-15 -15 -15>, angularTarget=<1 0 0>}
      localJointOffset=<4 0 0>
     }]
   }
   n11 = Anim Dynamics - Bone: dyn_sail_1_jnt!(ComponentPose=@n10, Alpha=1, GravityOverride:Vector="0,0,-980", ExternalForce="0,0,0") -> (Pose) {
    angularDampingOverride=1
    angularSpringConstant=20000
    bAngularSpring=true
    bOverrideAngularDamping=true
    bUseGravityOverride=true
    gravityScale=0.35
    simulationSpace=RootRelative
    alphaBoolBlend {blendOption=Linear}
    boundBone {boneName=dyn_sail_1_jnt}
    componentLinearAccScale=<1 1 1>
    componentLinearVelScale=<1 1 1>
    gravityOverride {z=-980}
    physicsBodyDefinitions [{
      boundBone {boneName=dyn_sail_1_jnt}
      boxExtents=<30 30 30>
      constraintSetup {angularLimitsMax=<15 15 15>, angularLimitsMin=<-15 -15 -15>, angularTarget=<1 0 0>}
      localJointOffset=<4 0 0>
     }]
   }
   n12 = Anim Dynamics - Bone: dyn_sail_2_jnt!(ComponentPose=@n11, Alpha=1, GravityOverride:Vector="0,0,-980", ExternalForce="0,0,0") -> (Pose) {
    angularDampingOverride=1
    angularSpringConstant=20000
    bAngularSpring=true
    bOverrideAngularDamping=true
    bUseGravityOverride=true
    gravityScale=0.35
    simulationSpace=RootRelative
    alphaBoolBlend {blendOption=Linear}
    boundBone {boneName=dyn_sail_2_jnt}
    componentLinearAccScale=<1 1 1>
    componentLinearVelScale=<1 1 1>
    gravityOverride {z=-980}
    physicsBodyDefinitions [{
      boundBone {boneName=dyn_sail_2_jnt}
      boxExtents=<30 30 30>
      constraintSetup {angularLimitsMax=<15 15 15>, angularLimitsMin=<-15 -15 -15>, angularTarget=<1 0 0>}
      localJointOffset=<4 0 0>
     }]
   }
   n13 = Anim Dynamics - Bone: dyn_sail_3_jnt!(ComponentPose=@n12, Alpha=1, GravityOverride:Vector="0,0,-980", ExternalForce="0,0,0") -> (Pose) {
    angularDampingOverride=1
    angularSpringConstant=20000
    bAngularSpring=true
    bOverrideAngularDamping=true
    bUseGravityOverride=true
    gravityScale=0.35
    simulationSpace=RootRelative
    alphaBoolBlend {blendOption=Linear}
    boundBone {boneName=dyn_sail_3_jnt}
    componentLinearAccScale=<1 1 1>
    componentLinearVelScale=<1 1 1>
    gravityOverride {z=-980}
    physicsBodyDefinitions [{
      boundBone {boneName=dyn_sail_3_jnt}
      boxExtents=<30 30 30>
      constraintSetup {angularLimitsMax=<15 15 15>, angularLimitsMin=<-15 -15 -15>, angularTarget=<1 0 0>}
      localJointOffset=<4 0 0>
     }]
   }
   n14 = Anim Dynamics - Bone: dyn_sail_4_jnt!(ComponentPose=@n13, Alpha=1, GravityOverride:Vector="0,0,-980", ExternalForce="0,0,0") -> (Pose) {
    angularDampingOverride=1
    angularSpringConstant=20000
    bAngularSpring=true
    bOverrideAngularDamping=true
    bUseGravityOverride=true
    gravityScale=0.35
    simulationSpace=RootRelative
    alphaBoolBlend {blendOption=Linear}
    boundBone {boneName=dyn_sail_4_jnt}
    componentLinearAccScale=<1 1 1>
    componentLinearVelScale=<1 1 1>
    gravityOverride {z=-980}
    physicsBodyDefinitions [{
      boundBone {boneName=dyn_sail_4_jnt}
      boxExtents=<30 30 30>
      constraintSetup {angularLimitsMax=<15 15 15>, angularLimitsMin=<-15 -15 -15>, angularTarget=<1 0 0>}
      localJointOffset=<4 0 0>
     }]
   }
   n15 = Anim Dynamics - Bone: dyn_sail_5_jnt!(ComponentPose=@n14, Alpha=1, GravityOverride:Vector="0,0,-980", ExternalForce="0,0,0") -> (Pose) {
    angularDampingOverride=1
    angularSpringConstant=20000
    bAngularSpring=true
    bOverrideAngularDamping=true
    bUseGravityOverride=true
    gravityScale=0.35
    simulationSpace=RootRelative
    alphaBoolBlend {blendOption=Linear}
    boundBone {boneName=dyn_sail_5_jnt}
    componentLinearAccScale=<1 1 1>
    componentLinearVelScale=<1 1 1>
    gravityOverride {z=-980}
    physicsBodyDefinitions [{
      boundBone {boneName=dyn_sail_5_jnt}
      boxExtents=<30 30 30>
      constraintSetup {angularLimitsMax=<15 15 15>, angularLimitsMin=<-15 -15 -15>, angularTarget=<1 0 0>}
      localJointOffset=<4 0 0>
     }]
   }
   n16 = Anim Dynamics - Bone: dyn_sail_6_jnt!(ComponentPose=@n15, Alpha=1, GravityOverride:Vector="0,0,-980", ExternalForce="0,0,0") -> (Pose) {
    angularDampingOverride=1
    angularSpringConstant=20000
    bAngularSpring=true
    bOverrideAngularDamping=true
    bUseGravityOverride=true
    gravityScale=0.35
    simulationSpace=RootRelative
    alphaBoolBlend {blendOption=Linear}
    boundBone {boneName=dyn_sail_6_jnt}
    componentLinearAccScale=<1 1 1>
    componentLinearVelScale=<1 1 1>
    gravityOverride {z=-980}
    physicsBodyDefinitions [{
      boundBone {boneName=dyn_sail_6_jnt}
      boxExtents=<30 30 30>
      constraintSetup {angularLimitsMax=<15 15 15>, angularLimitsMin=<-15 -15 -15>, angularTarget=<1 0 0>}
      localJointOffset=<4 0 0>
     }]
   }
   n17 = Anim Dynamics - Bone: dyn_sail_7_jnt!(ComponentPose=@n16, Alpha=1, GravityOverride:Vector="0,0,-980", ExternalForce="0,0,0") -> (Pose) {
    angularDampingOverride=1
    angularSpringConstant=20000
    bAngularSpring=true
    bOverrideAngularDamping=true
    bUseGravityOverride=true
    gravityScale=0.35
    simulationSpace=RootRelative
    alphaBoolBlend {blendOption=Linear}
    boundBone {boneName=dyn_sail_7_jnt}
    componentLinearAccScale=<1 1 1>
    componentLinearVelScale=<1 1 1>
    gravityOverride {z=-980}
    physicsBodyDefinitions [{
      boundBone {boneName=dyn_sail_7_jnt}
      boxExtents=<30 30 30>
      constraintSetup {angularLimitsMax=<15 15 15>, angularLimitsMin=<-15 -15 -15>, angularTarget=<1 0 0>}
      localJointOffset=<4 0 0>
     }]
   }
   n18 = Anim Dynamics - Bone: dyn_sail_8_jnt!(ComponentPose=@n17, Alpha=1, GravityOverride:Vector="0,0,-980", ExternalForce="0,0,0") -> (Pose) {
    angularDampingOverride=1
    angularSpringConstant=20000
    bAngularSpring=true
    bOverrideAngularDamping=true
    bUseGravityOverride=true
    gravityScale=0.35
    simulationSpace=RootRelative
    alphaBoolBlend {blendOption=Linear}
    boundBone {boneName=dyn_sail_8_jnt}
    componentLinearAccScale=<1 1 1>
    componentLinearVelScale=<1 1 1>
    gravityOverride {z=-980}
    physicsBodyDefinitions [{
      boundBone {boneName=dyn_sail_8_jnt}
      boxExtents=<30 30 30>
      constraintSetup {angularLimitsMax=<0 15 15>, angularLimitsMin=<0 -15 -15>, angularTarget=<1 0 0>}
      localJointOffset=<4 0 0>
     }]
   }
   Save cached pose 'DynamicsSailCache'(Pose=Component To Local(ComponentPose=@n18)) {cacheName=DynamicsSailCache}
  ]}
 EventGraph {nodes [
   Event BlueprintUpdateAnimation()
   > n1 = Sequence!(execute) -> (then_0, then_1, then_2:Exec, then_3:Exec, then_4:Exec)
   n2 = Sequence!(execute=@n1.then_4) -> (then_0, then_1, then_2, then_3, then_4, then_5:Exec, then_6:Exec)
   n3 = TryGetPawnOwner()
   > n4 = Cast To FortPlayerPawn(Object=@n3)
   n5 = GetSkeletalMeshForPartType(self=@n4)
   > n6 = Branch(Condition=MatchesWildcard(SourceString=GetDisplayName(Object=@n5), Wildcard=*M_MED_DomeRoof*, SearchCase=CaseSensitive))
   > Set Is_M_MED_DomeRoof(Is_M_MED_DomeRoof:bool=true) -> (Output_Get:bool)
   > Set Is_OtherCharacters(Is_OtherCharacters:bool=false) -> (Output_Get:bool)
   Set Is_OtherCharacters(execute=@n6.else, Is_OtherCharacters=true)
   n7 = FindBodyType!(execute=@n2.then_1, Object:Object Reference=@n3) -> (True:Exec, False:Exec, OutBodyType:EFortCustomBodyType Enum=NONE)
   n8 = FindGender!(execute=@n2.then_2, Object:Object Reference=@n3) -> (True:Exec, False:Exec, OutGender:EFortCustomGender Enum=Invalid)
   Set Is_M_LRG_Characters(
    execute=@n2.then_3
    Is_M_LRG_Characters:bool=AND Boolean(
     A="Equal (Enum)"(A:EFortCustomBodyType Enum=@n7, B:EFortCustomBodyType Enum=Large)
     B="Equal (Enum)"(A:EFortCustomGender Enum=@n8, B:EFortCustomGender Enum=Male)
    )
   ) -> (Output_Get:bool)
   n9 = FindBodyType(execute=@n2.then_4, Object=@n3) -> (OutBodyType=NONE)
   n10 = FindGender(execute=@n2.then_5, Object=@n3) -> (OutGender=Invalid)
   Set Is_F_MED_Characters(
    execute=@n2.then_6
    Is_F_MED_Characters:bool=AND Boolean(
     A="Equal (Enum)"(A:EFortCustomBodyType Enum=@n9, B:EFortCustomBodyType Enum=Medium)
     B="Equal (Enum)"(A:EFortCustomGender Enum=@n10, B:EFortCustomGender Enum=Female)
    )
   ) -> (Output_Get:bool)
   Set DBNO_Bottle_Rotate(execute=@n1.then_3, DBNO_Bottle_Rotate:double="To Float (Boolean)"(InBool=Get bIsDBNO())) -> (Output_Get:double)
   n11 = Gravity Override State Function DEPRECATED(
    execute=@n1.then_0
    CharacterPartAnimInstance=Self-Reference()
    GameTimeInSeconds=GetGameTimeInSeconds()
    GravityOverrideParams=Get Gravity Override Params()
   )
   > Set GravityOverride(GravityOverride=@n11)
   n12 = Branch(execute=@n1.then_1, Condition=Get bIsSkyDiving())
   > Set SkyDiving_ShipDyn_Alpha(SkyDiving_ShipDyn_Alpha:double=0.25) -> (Output_Get:double)
   Set SkyDiving_ShipDyn_Alpha(execute=@n12.else, SkyDiving_ShipDyn_Alpha=1)
   Set Floating_Bottle_Rotate(execute=@n1.then_2, Floating_Bottle_Rotate:double="To Float (Boolean)"(InBool=Get bIsSkydiveFloating())) -> (Output_Get:double)
   Event BlueprintPostEvaluateAnimation() {is=title K2Node_Event}
   n13 = GetPartSkeletalMeshComponent()
   > Is Valid(InputObject=@n13)
   > Delay(Duration=0)
   > SetVectorParameterValueOnMaterials!(
    execute
    self:Mesh Component Object Reference=@n13
    ParameterName:Name=ShipCenter
    ParameterValue:Vector=GetSocketLocation!(self:Scene Component Object Reference=@n13, InSocketName:Name=ShipCenter) -> (ReturnValue:Vector)
   ) -> (then)
   Event OnParachuteOpened()
   n14 = GetPartSkeletalMeshComponent()
   > n15 = Get Dynamic Material Instance(SkeletalMeshComponent=@n14)
   > Set Material_Water0(Material_Water0:Material Instance Dynamic Object Reference=@n15) -> (Output_Get:Material Instance Dynamic Object Reference)
   > n16 = Get Dynamic Material Instance(SkeletalMeshComponent=@n14, ElementIndex=1)
   > Set Material_Glass1(Material_Glass1:Material Instance Dynamic Object Reference=@n16) -> (Output_Get:Material Instance Dynamic Object Reference)
   n17 = GetPartSkeletalMeshComponent()
   > n18 = Get Dynamic Material Instance(SkeletalMeshComponent=@n17, ElementIndex=2)
   > Set Material_Ship2(Material_Ship2:Material Instance Dynamic Object Reference=@n18) -> (Output_Get:Material Instance Dynamic Object Reference)
   > n19 = GetOrCreateDynamicOverlayMaterial(MeshComponent=@n17, MaterialName=None)
   > Set OverlayMaterial_Glass0(OverlayMaterial_Glass0:Material Instance Dynamic Object Reference=@n19) -> (Output_Get:Material Instance Dynamic Object Reference)
   > n20 = Branch(Condition=GliderItemDefinitionHasAnyMetaTag!(Tags:Array of Gameplay Tag Structures=Make Array("[0]"=(TagName="Cosmetics.MetaTags.SpringBreak")), bRequiresOpened:bool=true) -> (ReturnValue:bool))
   > n21 = Sequence!(execute) -> (then_0, then_1)
   n22 = Get Material_Water0() -> (Material_Water0:Material Instance Dynamic Object Reference)
   SetScalarParameterValue(execute=@n21.then_1, self=@n22, ParameterName=Hide Element 10, Value=1)
   n23 = Get Material_Ship2() -> (Material_Ship2:Material Instance Dynamic Object Reference)
   > SetScalarParameterValue(self=@n23, ParameterName=Hide Element 10, Value=1)
   n24 = Get Material_Glass1() -> (Material_Glass1:Material Instance Dynamic Object Reference)
   > SetScalarParameterValue(self=@n24, ParameterName=HT_CrunchVerts, Value=1)
   n25 = Get OverlayMaterial_Glass0() -> (OverlayMaterial_Glass0:Material Instance Dynamic Object Reference)
   > SetScalarParameterValue(self=@n25, ParameterName=HT_CrunchVerts, Value=1)
   SpawnSystemAttached!(
    execute=@n21.then_0
    SystemTemplate:Niagara System Object Reference=/BRCosmetics/Accessories/FORT_Backpacks/Backpack_M_MED_DomeRoof/FX/NS_Backpack_DomeRoof_GlassBreak
    AttachToComponent:Scene Component Object Reference=GetPartSkeletalMeshComponent()
    AttachPointName:Name=pack
    Location:Vector="0, 0, 0"
    Rotation:Rotator="0, 0, 0"
    LocationType:EAttachLocation Enum=KeepRelativeOffset
    bAutoDestroy:bool=true
    bAutoActivate:bool=true
    PoolingMethod:ENCPoolMethod Enum=None
    bPreCullCheck:bool=true
   ) -> (then, ReturnValue:Niagara Particle System Component Object Reference)
   SetScalarParameterValue(execute=[@n26.then, @n27.then, @n20.else], self=@n22, ParameterName=Hide Element 10, Value=0)
   > SetScalarParameterValue(self=@n23, ParameterName=Hide Element 10, Value=0)
   > SetScalarParameterValue(self=@n24, ParameterName=HT_CrunchVerts, Value=0)
   > SetScalarParameterValue(self=@n25, ParameterName=HT_CrunchVerts, Value=0)
   n26 = Event OnBeginSkydiving()
   n27 = Event OnParachuteClosed()
   TryGetPawnOwner()
  ]}
]
variables [
 DBNO_Bottle_Rotate {type=number}
 Floating_Bottle_Rotate {type=number}
 Gravity Override Params {type=GravityOverrideParamsStruct}
 GravityOverride {type=Vector}
 Is_F_MED_Characters {type=boolean}
 Is_M_LRG_Characters {type=boolean}
 Is_M_MED_DomeRoof {type=boolean}
 Is_OtherCharacters {type=boolean}
 Material_Glass1 {type=MaterialInstanceDynamic}
 Material_Ship2 {type=MaterialInstanceDynamic}
 Material_Water0 {type=MaterialInstanceDynamic}
 OverlayMaterial_Glass0 {type=MaterialInstanceDynamic}
 SkyDiving_ShipDyn_Alpha {type=number}
]
